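# Control Rig for the jellyfish backpack cosmetic that procedurally animates its tentacles with wave motion, without inputs from variables.
# Two BoneHarmonics units rotate the bottom and top tentacle bones on Y, using per-bone ratios to vary the motion.
# Five ChainHarmonics units then ripple the long tentacle chains, using wave only (pendulum and reach off) with staggered offsets to desync them.
# The fish bones in the hierarchy are not driven by this graph.
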ControlRigBlueprint /BRCosmetics/Accessories/FORT_Backpacks/F_MED_Jellyfish_Backpack/Meshes/F_Med_Jellyfish_Backpack_CtrlRig
graphs [
 RigVMFunctionLibrary {links [], nodes []}
 RigVMModel {nodes [
   n1 = &RigUnit_BeginExecution()
   /Script/ControlRig.RigUnit_BoneHarmonics(
    Bones="((Bone=\"bottom0_0\",Ratio=1),(Bone=\"bottom1_0\",Ratio=0.1),(Bone=\"bottom2_0\",Ratio=0.2),(Bone=\"bottom3_0\",Ratio=0.3),(Bone=\"bottom4_0\",Ratio=0.4),(Bone=\"bottom5_0\",Ratio=0.5),(Bone=\"bottom6_0\",Ratio=0.6),(Bone=\"bottom7_0\",Ratio=0.7),(Bone=\"bottom8_0\",Ratio=0 ...(truncated)"
    RotationOrder=YZX
    WaveAmplitude=(X=0,Y=90,Z=0)
    WaveEase=Linear
    WaveFrequency=(X=1,Y=0.6,Z=0.8)
    WaveMaximum=1.2
    WaveMinimum=0.5
    WaveNoise=(X=0.2,Y=0.2,Z=0.2)
    WaveOffset=(X=0,Y=1,Z=2)
    WaveSpeed=(X=0.3,Y=0.3,Z=0.3)
    bPropagateToChildren=False
    ExecutePin=@n1.ExecutePin
   ) {is=struct RigVMUnitNode}
   > &RigUnit_BoneHarmonics(
    Bones="((Bone=\"top0_0\",Ratio=0),(Bone=\"top1_0\",Ratio=0.1),(Bone=\"top2_0\",Ratio=0.2),(Bone=\"top3_0\",Ratio=0.3),(Bone=\"top4_0\",Ratio=0.4),(Bone=\"top5_0\",Ratio=0.5),(Bone=\"top6_0\",Ratio=0.56),(Bone=\"top7_0\",Ratio=0.7),(Bone=\"top8_0\",Ratio=0.8),(Bone=\"top9_0\",Ratio=0 ...(truncated)"
    RotationOrder=YZX
    WaveAmplitude=(X=0,Y=70,Z=0)
    WaveEase=Linear
    WaveFrequency=(X=1,Y=1,Z=1)
    WaveMaximum=1
    WaveMinimum=0.5
    WaveNoise=(X=0.2,Y=0.2,Z=0.2)
    WaveOffset=(X=0,Y=1,Z=2)
    WaveSpeed=(X=0.3,Y=0.3,Z=0.3)
    bPropagateToChildren=False
   )
   > /Script/ControlRig.RigUnit_ChainHarmonics(
    ChainRoot=long0_0
    DrawWorldOffset=(Rotation=(X=0,Y=0,Z=0,W=1),Translation=(X=0,Y=0,Z=0),Scale3D=(X=1,Y=1,Z=1))
    Pendulum=(bEnabled=False,PendulumStiffness=2,PendulumGravity=(X=0,Y=0,Z=-0.98),PendulumBlend=0.75,PendulumDrag=0.98,PendulumMinimum=0,PendulumMaximum=0.1,PendulumEase=Linear,UnwindAxis=(X=0,Y=1,Z=0),UnwindMinimum=0.2,UnwindMaximum=0.05)
    Reach=(bEnabled=False,ReachTarget=(X=0,Y=0,Z=0),ReachAxis=(X=1,Y=0,Z=0),ReachMinimum=0,ReachMaximum=0,ReachEase=Linear)
    Speed=(X=2,Y=0.5,Z=0.5)
    Wave=(bEnabled=True,WaveFrequency=(X=1,Y=1,Z=5),WaveAmplitude=(X=0,Y=0.4,Z=0.4),WaveOffset=(X=0,Y=0,Z=0),WaveNoise=(X=0,Y=0.2,Z=0.2),WaveMinimum=0,WaveMaximum=1,WaveEase=Linear)
    WaveCurve=(EditorCurveData=(Keys=((Time=0.0452,Value=0.7355),(Time=0.3862,Value=0.785),(Time=0.8152,Value=0.9115),(Time=1,Value=1)),PreInfinityExtrap=RCCE_Constant,PostInfinityExtrap=RCCE_Constant,DefaultValue=340282346638528859811704183484516925440))
    bDrawDebug=false
   ) {is=struct RigVMUnitNode}
   > &RigUnit_ChainHarmonics(
    ChainRoot=long1_0
    DrawWorldOffset=(Rotation=(X=0,Y=0,Z=0,W=1),Translation=(X=0,Y=0,Z=0),Scale3D=(X=1,Y=1,Z=1))
    Pendulum=(bEnabled=False,PendulumStiffness=2,PendulumGravity=(X=0,Y=0,Z=-0.98),PendulumBlend=0.75,PendulumDrag=0.98,PendulumMinimum=0,PendulumMaximum=0.1,PendulumEase=Linear,UnwindAxis=(X=0,Y=1,Z=0),UnwindMinimum=0.2,UnwindMaximum=0.05)
    Reach=(bEnabled=False,ReachTarget=(X=0,Y=0,Z=0),ReachAxis=(X=1,Y=0,Z=0),ReachMinimum=0,ReachMaximum=0,ReachEase=Linear)
    Speed=(X=2,Y=0.5,Z=0.5)
    Wave=(bEnabled=True,WaveFrequency=(X=1,Y=1,Z=5),WaveAmplitude=(X=0,Y=0.4,Z=0.4),WaveOffset=(X=0,Y=1,Z=1),WaveNoise=(X=0,Y=0.2,Z=0.2),WaveMinimum=0,WaveMaximum=1,WaveEase=Linear)
    WaveCurve=(EditorCurveData=(Keys=((Time=0.0452,Value=0.7355),(Time=0.3862,Value=0.785),(Time=0.8152,Value=0.9115),(Time=1,Value=1)),PreInfinityExtrap=RCCE_Constant,PostInfinityExtrap=RCCE_Constant,DefaultValue=340282346638528859811704183484516925440))
    bDrawDebug=false
   )
   > &RigUnit_ChainHarmonics(
    ChainRoot=long2_0
    DrawWorldOffset=(Rotation=(X=0,Y=0,Z=0,W=1),Translation=(X=0,Y=0,Z=0),Scale3D=(X=1,Y=1,Z=1))
    Pendulum=(bEnabled=False,PendulumStiffness=2,PendulumGravity=(X=0,Y=0,Z=-0.98),PendulumBlend=0.75,PendulumDrag=0.98,PendulumMinimum=0,PendulumMaximum=0.1,PendulumEase=Linear,UnwindAxis=(X=0,Y=1,Z=0),UnwindMinimum=0.2,UnwindMaximum=0.05)
    Reach=(bEnabled=False,ReachTarget=(X=0,Y=0,Z=0),ReachAxis=(X=1,Y=0,Z=0),ReachMinimum=0,ReachMaximum=0,ReachEase=Linear)
    Speed=(X=2,Y=0.5,Z=0.5)
    Wave=(bEnabled=True,WaveFrequency=(X=1,Y=1,Z=5),WaveAmplitude=(X=0,Y=0.4,Z=0.4),WaveOffset=(X=0,Y=2,Z=2),WaveNoise=(X=0,Y=0.2,Z=0.2),WaveMinimum=0,WaveMaximum=1,WaveEase=Linear)
    WaveCurve=(EditorCurveData=(Keys=((Time=0.0452,Value=0.7355),(Time=0.3862,Value=0.785),(Time=0.8152,Value=0.9115),(Time=1,Value=1)),PreInfinityExtrap=RCCE_Constant,PostInfinityExtrap=RCCE_Constant,DefaultValue=340282346638528859811704183484516925440))
    bDrawDebug=false
   )
   > &RigUnit_ChainHarmonics(
    ChainRoot=long3_0
    DrawWorldOffset=(Rotation=(X=0,Y=0,Z=0,W=1),Translation=(X=0,Y=0,Z=0),Scale3D=(X=1,Y=1,Z=1))
    Pendulum=(bEnabled=False,PendulumStiffness=2,PendulumGravity=(X=0,Y=0,Z=-0.98),PendulumBlend=0.75,PendulumDrag=0.98,PendulumMinimum=0,PendulumMaximum=0.1,PendulumEase=Linear,UnwindAxis=(X=0,Y=1,Z=0),UnwindMinimum=0.2,UnwindMaximum=0.05)
    Reach=(bEnabled=False,ReachTarget=(X=0,Y=0,Z=0),ReachAxis=(X=1,Y=0,Z=0),ReachMinimum=0,ReachMaximum=0,ReachEase=Linear)
    Speed=(X=2,Y=0.5,Z=0.5)
    Wave=(bEnabled=True,WaveFrequency=(X=1,Y=1,Z=5),WaveAmplitude=(X=0,Y=0.4,Z=0.4),WaveOffset=(X=0,Y=3,Z=3),WaveNoise=(X=0,Y=0.2,Z=0.2),WaveMinimum=0,WaveMaximum=1,WaveEase=Linear)
    WaveCurve=(EditorCurveData=(Keys=((Time=0.0452,Value=0.7355),(Time=0.3862,Value=0.785),(Time=0.8152,Value=0.9115),(Time=1,Value=1)),PreInfinityExtrap=RCCE_Constant,PostInfinityExtrap=RCCE_Constant,DefaultValue=340282346638528859811704183484516925440))
    bDrawDebug=false
   )
   > &RigUnit_ChainHarmonics(
    ChainRoot=long4_0
    DrawWorldOffset=(Rotation=(X=0,Y=0,Z=0,W=1),Translation=(X=0,Y=0,Z=0),Scale3D=(X=1,Y=1,Z=1))
    Pendulum=(bEnabled=False,PendulumStiffness=2,PendulumGravity=(X=0,Y=0,Z=-0.98),PendulumBlend=0.75,PendulumDrag=0.98,PendulumMinimum=0,PendulumMaximum=0.1,PendulumEase=Linear,UnwindAxis=(X=0,Y=1,Z=0),UnwindMinimum=0.2,UnwindMaximum=0.05)
    Reach=(bEnabled=False,ReachTarget=(X=0,Y=0,Z=0),ReachAxis=(X=1,Y=0,Z=0),ReachMinimum=0,ReachMaximum=0,ReachEase=Linear)
    Speed=(X=2,Y=0.5,Z=0.5)
    Wave=(bEnabled=True,WaveFrequency=(X=1,Y=1,Z=5),WaveAmplitude=(X=0,Y=0.4,Z=0.4),WaveOffset=(X=0,Y=4,Z=4),WaveNoise=(X=0,Y=0.2,Z=0.2),WaveMinimum=0,WaveMaximum=1,WaveEase=Linear)
    WaveCurve=(EditorCurveData=(Keys=((Time=0.0452,Value=0.7355),(Time=0.3862,Value=0.785),(Time=0.8152,Value=0.9115),(Time=1,Value=1)),PreInfinityExtrap=RCCE_Constant,PostInfinityExtrap=RCCE_Constant,DefaultValue=340282346638528859811704183484516925440))
    bDrawDebug=false
   )
  ]}
]
hierarchy [
 root {loc=<0 0 0>, rot=<0 0 0 1>}
 pack {parent=root, loc=<0 -0.7316 0>, rot=<0 0 0 1>}
 tent_base {parent=pack, loc=<0.6952 -3.5807 0.4861>, rot=<0 0 0 1>}
 bottom_world {parent=tent_base, loc=<-2.287 -0.3621 -16.0602>, rot=<0 0 0 1>}
 bottom_base {parent=bottom_world, loc=<0 0 0>, rot=<0.7033 -0.5525 -0.0731 0.4413>}
 bottom16_0 {parent=bottom_base, loc=<8.2232 4.3353 -1.1525>, rot=<-0.3153 -0.1425 -0.2812 -0.8951>}
 bottom15_0 {parent=bottom_base, loc=<10.017 0.8232 0.5206>, rot=<-0.1537 0.0329 -0.0915 -0.9833>}
 bottom14_0 {parent=bottom_base, loc=<8.8668 2.603 0.6204>, rot=<-0.2967 -0.0101 -0.1949 -0.9348>}
 bottom13_0 {parent=bottom_base, loc=<10.7694 0.7668 -1.7108>, rot=<0 0 0 -1>}
 bottom12_0 {parent=bottom_base, loc=<8.7003 -0.5676 -3.3467>, rot=<0.0687 -0.2916 0.0812 -0.9506>}
 bottom11_0 {parent=bottom_base, loc=<9.6986 -0.1135 -2.3225>, rot=<0.0719 -0.1139 0.0588 -0.9891>}
 bottom10_0 {parent=bottom_base, loc=<10.4324 1.6353 -0.8483>, rot=<-0.2967 -0.0101 -0.1949 -0.9348>}
 bottom9_0 {parent=bottom_base, loc=<9.1136 2.7679 -3.5357>, rot=<-0.1002 -0.1694 -0.0919 -0.9761>}
 bottom8_0 {parent=bottom_base, loc=<8.6712 3.0147 -0.5554>, rot=<-0.3224 -0.1223 -0.2748 -0.8975>}
 bottom7_0 {parent=bottom_base, loc=<9.4261 2.6796 -1.5314>, rot=<-0.0436 -0.1795 -0.0407 -0.9819>}
 bottom6_0 {parent=bottom_base, loc=<9.8656 -1.3002 -0.7025>, rot=<0.1981 0.0077 0.1249 -0.9722>}
 bottom5_0 {parent=bottom_base, loc=<9.4987 -0.2592 -0.0397>, rot=<0.3699 0.1118 0.1775 -0.9051>}
 bottom4_0 {parent=bottom_base, loc=<7.7991 0.7093 -3.8818>, rot=<-0.0371 -0.3492 -0.0495 -0.935>}
 bottom3_0 {parent=bottom_base, loc=<8.3636 -1.3345 -2.0105>, rot=<0.2756 -0.1972 0.2761 -0.8994>}
 bottom2_0 {parent=bottom_base, loc=<9.3918 3.4716 -2.2872>, rot=<-0.2124 -0.2661 -0.2444 -0.9079>}
 bottom1_0 {parent=bottom_base, loc=<7.6997 3.9696 -0.3956>, rot=<-0.4774 -0.1045 -0.4062 -0.7722>}
 bottom0_0 {parent=bottom_base, loc=<10.1433 1.1691 -3.2856>, rot=<0.1065 -0.2139 0.1072 -0.9651>}
 top_world {parent=tent_base, loc=<0.2015 0.2789 0.2417>, rot=<0 0 0 1>}
 top_base {parent=top_world, loc=<0 0 0>, rot=<-0.569 0.5127 -0.344 -0.5432>}
 top23_0 {parent=top_base, loc=<8.0991 4.133 -6.0382>, rot=<-0.0429 0.0631 0.0607 -0.9952>}
 top22_0 {parent=top_base, loc=<7.7667 7.9948 -5.7534>, rot=<-0.0367 -0.0034 0.1027 -0.994>}
 top21_0 {parent=top_base, loc=<11.8363 -4.95 -6.7701>, rot=<-0.1237 0.0558 -0.2546 -0.9575>}
 top20_0 {parent=top_base, loc=<7.6288 -8.6909 -0.3492>, rot=<-0.0601 -0.1608 -0.0432 -0.9842>}
 top19_0 {parent=top_base, loc=<4.8783 -10.5386 0.8673>, rot=<-0.1693 -0.3444 0.6083 -0.6948>}
 top18_0 {parent=top_base, loc=<9.8777 -5.6089 5.6633>, rot=<-0.0588 -0.2686 0.012 -0.9614>}
 top17_0 {parent=top_base, loc=<9.6878 -0.1315 -5.8808>, rot=<-0.1085 -0.0078 -0.2748 -0.9553>}
 top16_0 {parent=top_base, loc=<9.0204 8.2684 5.7968>, rot=<-0.0836 -0.034 -0.15 -0.9846>}
 top15_0 {parent=top_base, loc=<5.3816 -2.4468 9.3281>, rot=<-0.0627 -0.0292 -0.0204 -0.9974>}
 top14_0 {parent=top_base, loc=<9.2383 -8.8333 2.8572>, rot=<-0.0466 -0.1308 0.127 -0.9821>}
 top13_0 {parent=top_base, loc=<10.0651 3.6901 5.95>, rot=<-0.0793 -0.077 -0.1764 -0.9781>}
 top12_0 {parent=top_base, loc=<11.2694 1.2454 0.1198>, rot=<-0.0878 -0.0879 -0.3396 -0.9323>}
 top11_0 {parent=top_base, loc=<9.9876 -2.7157 -1.8403>, rot=<-0.0677 -0.1457 -0.253 -0.954>}
 top10_0 {parent=top_base, loc=<9.366 7.5068 1.6791>, rot=<-0.0697 -0.1237 -0.1605 -0.9768>}
 top9_0 {parent=top_base, loc=<8.7141 5.86 8.3844>, rot=<-0.078 -0.0835 -0.176 -0.9777>}
 top8_0 {parent=top_base, loc=<5.9399 10.519 4.4191>, rot=<-0.0577 -0.2017 -0.1186 -0.9705>}
 top7_0 {parent=top_base, loc=<6.0875 -8.4745 6.0268>, rot=<0.1883 0.0585 0.6494 -0.7345>}
 top6_0 {parent=top_base, loc=<5.9377 10.5776 0.4006>, rot=<-0.0741 -0.1144 -0.4873 -0.8625>}
 top5_0 {parent=top_base, loc=<9.921 -4.2191 -5.0025>, rot=<-0.0432 -0.2302 -0.2758 -0.9322>}
 top4_0 {parent=top_base, loc=<8.5698 -7.4829 -3.996>, rot=<-0.0704 -0.0998 -0.1202 -0.9852>}
 top3_0 {parent=top_base, loc=<8.2593 7.4233 -1.6954>, rot=<-0.0499 -0.2301 -0.1951 -0.9521>}
 top2_0 {parent=top_base, loc=<9.7156 -2.2876 3.8652>, rot=<-0.0492 -0.0919 0.0728 -0.9919>}
 top1_0 {parent=top_base, loc=<6.7338 2.1616 10.4762>, rot=<-0.1162 -0.0347 -0.4401 -0.8897>}
 top0_0 {parent=top_base, loc=<7.0459 11.5872 -2.5187>, rot=<-0.0797 -0.0002 -0.0996 -0.9918>}
 long_world {parent=tent_base, loc=<-5.5002 1.4218 -16.4717>, rot=<0 0 0 1>}
 long_base {parent=long_world, loc=<0 0 0>, rot=<0.6033 -0.6892 0.3689 0.1583>}
 long0_0 {parent=long_base, loc=<3.5129 1.3667 1.5933>, rot=<-0.253 0.0359 -0.2294 -0.9392>}
 long0_1 {parent=long0_0, loc=<7.6488 0 0>, rot=<0.1437 -0.2966 0.2295 -0.9158>}
 long0_2 {parent=long0_1, loc=<4.5522 0 0>, rot=<0.029 -0.1605 0.4287 -0.8886>}
 long0_3 {parent=long0_2, loc=<4.3863 0 0>, rot=<-0.0055 0.1272 -0.2634 -0.9563>}
 long1_0 {parent=long_base, loc=<3.9225 1.1976 1.0124>, rot=<-0.0294 -0.1193 -0.0366 -0.9918>}
 long1_1 {parent=long1_0, loc=<4.7766 0 0>, rot=<-0.0379 -0.2774 -0.1356 -0.9504>}
 long1_2 {parent=long1_1, loc=<4.5312 0 0>, rot=<-0.0101 -0.0385 0.2002 -0.9789>}
 long1_3 {parent=long1_2, loc=<5.0688 0 0>, rot=<-0.0019 0.0797 0.2604 -0.9622>}
 long3_0 {parent=long_base, loc=<3.165 0.6992 1.0297>, rot=<0 0 0 1>}
 long3_1 {parent=long3_0, loc=<6.2795 0 0>, rot=<-0.0508 -0.309 -0.0975 -0.9447>}
 long3_2 {parent=long3_1, loc=<7.8087 0 0>, rot=<0.003 -0.1459 0.2119 -0.9663>}
 long3_3 {parent=long3_2, loc=<7.3022 0 0>, rot=<-0.0058 0.0982 0.1613 -0.982>}
 long4_0 {parent=long_base, loc=<3.2054 -0.5319 0.5102>, rot=<0.0516 -0.1096 0.063 -0.9906>}
 long4_1 {parent=long4_0, loc=<9.1316 0 0>, rot=<-0.0269 -0.0952 -0.0516 -0.9938>}
 long4_2 {parent=long4_1, loc=<7.5246 0 0>, rot=<-0.1541 0.0839 -0.3004 -0.9375>}
 long4_3 {parent=long4_2, loc=<5.0035 0 0>, rot=<0.0939 -0.2642 0.3286 -0.9019>}
 long2_0 {parent=long_base, loc=<3.4933 0.247 1.7898>, rot=<0.0952 0.0555 0.0832 -0.9904>}
 long2_1 {parent=long2_0, loc=<4.3649 0 0>, rot=<-0.0357 -0.3481 -0.0581 -0.935>}
 long2_2 {parent=long2_1, loc=<4.0727 0 0>, rot=<-0.0111 -0.0306 -0.0668 -0.9972>}
 long2_3 {parent=long2_2, loc=<4.3669 0 0>, rot=<-0.0252 0.1458 -0.0873 -0.9851>}
 fish2_world {parent=pack, loc=<1.8967 -15.248 6.791>, rot=<0 0 0 1>}
 fish2_0 {parent=fish2_world, loc=<0 0 0>, rot=<-0.0564 -0.0263 0.9088 -0.4125>}
 fish1_world {parent=pack, loc=<-4.9061 -9.8553 5.3649>, rot=<0 0 0 1>}
 fish1_0 {parent=fish1_world, loc=<0 0 0>, rot=<-0.0622 0.0024 0.9969 0.0491>}
 fish3_world {parent=pack, loc=<10.624 -11.5672 6.2716>, rot=<0 0 0 1>}
 fish3_0 {parent=fish3_world, loc=<0 0 0>, rot=<-0.0104 -0.0223 0.4226 -0.906>}
]
