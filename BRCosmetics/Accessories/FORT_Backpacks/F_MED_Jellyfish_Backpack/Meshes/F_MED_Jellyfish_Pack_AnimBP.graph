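# Anim Blueprint for the jellyfish back bling, simulating secondary motion on its bell, tentacles and fish.
# Anim Dynamics drives lead top, bottom, long_world and fish bones; rotation constraints copy the top and bottom leads to sibling bones at varied alphas.
# A Control Rig follows, then Anim Dynamics on long-tentacle _2/_3 segments; _0/_1 nodes never reach the output.
# `skydiving_alpha` is 0 while skydiving, disabling all dynamics except the fish.
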AnimBlueprint /BRCosmetics/Accessories/FORT_Backpacks/F_MED_Jellyfish_Backpack/Meshes/F_MED_Jellyfish_Pack_AnimBP
target_skeleton=&/F_MED_Jellyfish_Pack_Skeleton
graphs [
 AnimGraph {nodes [
   n1 = Get skydiving_alpha() -> (skydiving_alpha:double)
   n2 = Anim Dynamics - Bone: long4_0(Alpha=@n1, ExternalForce="0,0,0") {
    angularDampingOverride=1
    angularSpringConstant=1000
    bAngularSpring=true
    bGravityOverrideInSimSpace=true
    bOverrideAngularDamping=true
    bUsePlanarLimit=false
    bUseSphericalLimits=true
    gravityScale=0
    alphaBoolBlend {blendOption=Linear}
    boundBone {boneName=long4_0}
    componentAppliedLinearAccClamp=<10000 10000 10000>
    componentLinearAccScale=<1 1 1>
    componentLinearVelScale=<1 1 1>
    physicsBodyDefinitions [{
      sphereCollisionRadius=0
      boundBone {boneName=long4_0}
      boxExtents=<50 50 20>
      constraintSetup {coneAngle=20, twistAxis=AxisZ, angularLimitsMax=<10 30 30>, angularLimitsMin=<-10 -30 -30>, angularTarget=<1 0 0>}
      localJointOffset=<10 0 0>
     }]
    planarLimits [{drivingBone {boneName=pack}, planeTransform {location=<0 0 0>, rotation=<rot 90 120 0>}}]
    relativeSpaceBone {boneName=root}
    sphericalLimits [
     {limitRadius=18, drivingBone {boneName=pack}, sphereLocalOffset=<5 5 -3>}
     {limitRadius=25, drivingBone {boneName=pack}, sphereLocalOffset=<-10 20 -3>}
    ]
   }
   Anim Dynamics - Bone: long4_1(ComponentPose=@n2, Alpha=@n1, ExternalForce="0,0,0") {
    angularDampingOverride=1
    angularSpringConstant=1000
    bAngularSpring=true
    bGravityOverrideInSimSpace=true
    bOverrideAngularDamping=true
    bUsePlanarLimit=false
    bUseSphericalLimits=true
    gravityScale=0
    alphaBoolBlend {blendOption=Linear}
    boundBone {boneName=long4_1}
    componentAppliedLinearAccClamp=<10000 10000 10000>
    componentLinearAccScale=<1 1 1>
    componentLinearVelScale=<1 1 1>
    physicsBodyDefinitions [{
      sphereCollisionRadius=0
      boundBone {boneName=long4_1}
      boxExtents=<50 50 20>
      constraintSetup {coneAngle=20, twistAxis=AxisZ, angularLimitsMax=<10 30 30>, angularLimitsMin=<-10 -30 -30>, angularTarget=<1 0 0>}
      localJointOffset=<10 0 0>
     }]
    relativeSpaceBone {boneName=root}
    sphericalLimits [
     {limitRadius=18, drivingBone {boneName=pack}, sphereLocalOffset=<5 5 -3>}
     {limitRadius=25, drivingBone {boneName=pack}, sphereLocalOffset=<-10 20 -3>}
    ]
   }
   n3 = Anim Dynamics - Bone: long0_0(Alpha=@n1, ExternalForce="0,0,0") {
    angularDampingOverride=1
    angularSpringConstant=1000
    bAngularSpring=true
    bGravityOverrideInSimSpace=true
    bOverrideAngularDamping=true
    bUsePlanarLimit=false
    bUseSphericalLimits=true
    gravityScale=0
    alphaBoolBlend {blendOption=Linear}
    boundBone {boneName=long0_0}
    componentAppliedLinearAccClamp=<10000 10000 10000>
    componentLinearAccScale=<1 1 1>
    componentLinearVelScale=<1 1 1>
    physicsBodyDefinitions [{
      sphereCollisionRadius=0
      boundBone {boneName=long0_0}
      boxExtents=<50 50 20>
      constraintSetup {coneAngle=20, twistAxis=AxisZ, angularLimitsMax=<10 30 30>, angularLimitsMin=<-10 -30 -30>, angularTarget=<1 0 0>}
      localJointOffset=<10 0 0>
     }]
    relativeSpaceBone {boneName=root}
    sphericalLimits [
     {limitRadius=18, drivingBone {boneName=pack}, sphereLocalOffset=<5 5 -3>}
     {limitRadius=25, drivingBone {boneName=pack}, sphereLocalOffset=<-10 20 -3>}
    ]
   }
   Anim Dynamics - Bone: long0_1(ComponentPose=@n3, Alpha=@n1, ExternalForce="0,0,0") {
    angularDampingOverride=1
    angularSpringConstant=1000
    bAngularSpring=true
    bGravityOverrideInSimSpace=true
    bOverrideAngularDamping=true
    bUsePlanarLimit=false
    bUseSphericalLimits=true
    gravityScale=0
    alphaBoolBlend {blendOption=Linear}
    boundBone {boneName=long0_1}
    componentAppliedLinearAccClamp=<10000 10000 10000>
    componentLinearAccScale=<1 1 1>
    componentLinearVelScale=<1 1 1>
    physicsBodyDefinitions [{
      sphereCollisionRadius=0
      boundBone {boneName=long0_1}
      boxExtents=<50 50 20>
      constraintSetup {coneAngle=20, twistAxis=AxisZ, angularLimitsMax=<10 30 30>, angularLimitsMin=<-10 -30 -30>, angularTarget=<1 0 0>}
      localJointOffset=<10 0 0>
     }]
    relativeSpaceBone {boneName=root}
    sphericalLimits [
     {limitRadius=18, drivingBone {boneName=pack}, sphereLocalOffset=<5 5 -3>}
     {limitRadius=25, drivingBone {boneName=pack}, sphereLocalOffset=<-10 20 -3>}
    ]
   }
   n4 = Anim Dynamics - Bone: long1_0(Alpha=@n1, ExternalForce="0,0,0") {
    angularDampingOverride=1
    angularSpringConstant=1000
    bAngularSpring=true
    bGravityOverrideInSimSpace=true
    bOverrideAngularDamping=true
    bUsePlanarLimit=false
    bUseSphericalLimits=true
    gravityScale=0
    alphaBoolBlend {blendOption=Linear}
    boundBone {boneName=long1_0}
    componentAppliedLinearAccClamp=<10000 10000 10000>
    componentLinearAccScale=<1 1 1>
    componentLinearVelScale=<1 1 1>
    physicsBodyDefinitions [{
      sphereCollisionRadius=0
      boundBone {boneName=long1_0}
      boxExtents=<50 50 20>
      constraintSetup {coneAngle=20, twistAxis=AxisZ, angularLimitsMax=<10 30 30>, angularLimitsMin=<-10 -30 -30>, angularTarget=<1 0 0>}
      localJointOffset=<10 0 0>
     }]
    relativeSpaceBone {boneName=root}
    sphericalLimits [
     {limitRadius=18, drivingBone {boneName=pack}, sphereLocalOffset=<5 5 -3>}
     {limitRadius=25, drivingBone {boneName=pack}, sphereLocalOffset=<-10 20 -3>}
    ]
   }
   Anim Dynamics - Bone: long1_1(ComponentPose=@n4, Alpha=@n1, ExternalForce="0,0,0") {
    angularDampingOverride=1
    angularSpringConstant=1000
    bAngularSpring=true
    bGravityOverrideInSimSpace=true
    bOverrideAngularDamping=true
    bUsePlanarLimit=false
    bUseSphericalLimits=true
    gravityScale=0
    alphaBoolBlend {blendOption=Linear}
    boundBone {boneName=long1_1}
    componentAppliedLinearAccClamp=<10000 10000 10000>
    componentLinearAccScale=<1 1 1>
    componentLinearVelScale=<1 1 1>
    physicsBodyDefinitions [{
      sphereCollisionRadius=0
      boundBone {boneName=long1_1}
      boxExtents=<50 50 20>
      constraintSetup {coneAngle=20, twistAxis=AxisZ, angularLimitsMax=<10 30 30>, angularLimitsMin=<-10 -30 -30>, angularTarget=<1 0 0>}
      localJointOffset=<10 0 0>
     }]
    relativeSpaceBone {boneName=root}
    sphericalLimits [
     {limitRadius=18, drivingBone {boneName=pack}, sphereLocalOffset=<5 5 -3>}
     {limitRadius=25, drivingBone {boneName=pack}, sphereLocalOffset=<-10 20 -3>}
    ]
   }
   n5 = Anim Dynamics - Bone: long2_0(Alpha=@n1, ExternalForce="0,0,0") {
    angularDampingOverride=1
    angularSpringConstant=1000
    bAngularSpring=true
    bGravityOverrideInSimSpace=true
    bOverrideAngularDamping=true
    bUsePlanarLimit=false
    bUseSphericalLimits=true
    gravityScale=0
    alphaBoolBlend {blendOption=Linear}
    boundBone {boneName=long2_0}
    componentAppliedLinearAccClamp=<10000 10000 10000>
    componentLinearAccScale=<1 1 1>
    componentLinearVelScale=<1 1 1>
    physicsBodyDefinitions [{
      sphereCollisionRadius=0
      boundBone {boneName=long2_0}
      boxExtents=<50 50 20>
      constraintSetup {coneAngle=20, twistAxis=AxisZ, angularLimitsMax=<10 30 30>, angularLimitsMin=<-10 -30 -30>, angularTarget=<1 0 0>}
      localJointOffset=<10 0 0>
     }]
    relativeSpaceBone {boneName=root}
    sphericalLimits [
     {limitRadius=18, drivingBone {boneName=pack}, sphereLocalOffset=<5 5 -3>}
     {limitRadius=25, drivingBone {boneName=pack}, sphereLocalOffset=<-10 20 -3>}
    ]
   }
   Anim Dynamics - Bone: long2_1(ComponentPose=@n5, Alpha=@n1, ExternalForce="0,0,0") {
    angularDampingOverride=1
    angularSpringConstant=1000
    bAngularSpring=true
    bGravityOverrideInSimSpace=true
    bOverrideAngularDamping=true
    bUsePlanarLimit=false
    bUseSphericalLimits=true
    gravityScale=0
    alphaBoolBlend {blendOption=Linear}
    boundBone {boneName=long2_1}
    componentAppliedLinearAccClamp=<10000 10000 10000>
    componentLinearAccScale=<1 1 1>
    componentLinearVelScale=<1 1 1>
    physicsBodyDefinitions [{
      sphereCollisionRadius=0
      boundBone {boneName=long2_1}
      boxExtents=<50 50 20>
      constraintSetup {coneAngle=20, twistAxis=AxisZ, angularLimitsMax=<10 30 30>, angularLimitsMin=<-10 -30 -30>, angularTarget=<1 0 0>}
      localJointOffset=<10 0 0>
     }]
    relativeSpaceBone {boneName=root}
    sphericalLimits [
     {limitRadius=18, drivingBone {boneName=pack}, sphereLocalOffset=<5 5 -3>}
     {limitRadius=25, drivingBone {boneName=pack}, sphereLocalOffset=<-10 20 -3>}
    ]
   }
   n6 = Anim Dynamics - Bone: long3_0(Alpha=@n1, ExternalForce="0,0,0") {
    angularDampingOverride=1
    angularSpringConstant=1000
    bAngularSpring=true
    bGravityOverrideInSimSpace=true
    bOverrideAngularDamping=true
    bUsePlanarLimit=false
    bUseSphericalLimits=true
    gravityScale=0
    alphaBoolBlend {blendOption=Linear}
    boundBone {boneName=long3_0}
    componentAppliedLinearAccClamp=<10000 10000 10000>
    componentLinearAccScale=<1 1 1>
    componentLinearVelScale=<1 1 1>
    physicsBodyDefinitions [{
      sphereCollisionRadius=0
      boundBone {boneName=long3_0}
      boxExtents=<50 50 20>
      constraintSetup {coneAngle=20, twistAxis=AxisZ, angularLimitsMax=<10 30 30>, angularLimitsMin=<-10 -30 -30>, angularTarget=<1 0 0>}
      localJointOffset=<10 0 0>
     }]
    relativeSpaceBone {boneName=root}
    sphericalLimits [
     {limitRadius=18, drivingBone {boneName=pack}, sphereLocalOffset=<5 5 -3>}
     {limitRadius=25, drivingBone {boneName=pack}, sphereLocalOffset=<-10 20 -3>}
    ]
   }
   Anim Dynamics - Bone: long3_1(ComponentPose=@n6, Alpha=@n1, ExternalForce="0,0,0") {
    angularDampingOverride=1
    angularSpringConstant=1000
    bAngularSpring=true
    bGravityOverrideInSimSpace=true
    bOverrideAngularDamping=true
    bUsePlanarLimit=false
    bUseSphericalLimits=true
    gravityScale=0
    alphaBoolBlend {blendOption=Linear}
    boundBone {boneName=long3_1}
    componentAppliedLinearAccClamp=<10000 10000 10000>
    componentLinearAccScale=<1 1 1>
    componentLinearVelScale=<1 1 1>
    physicsBodyDefinitions [{
      sphereCollisionRadius=0
      boundBone {boneName=long3_1}
      boxExtents=<50 50 20>
      constraintSetup {coneAngle=20, twistAxis=AxisZ, angularLimitsMax=<10 30 30>, angularLimitsMin=<-10 -30 -30>, angularTarget=<1 0 0>}
      localJointOffset=<10 0 0>
     }]
    relativeSpaceBone {boneName=root}
    sphericalLimits [
     {limitRadius=18, drivingBone {boneName=pack}, sphereLocalOffset=<5 5 -3>}
     {limitRadius=25, drivingBone {boneName=pack}, sphereLocalOffset=<-10 20 -3>}
    ]
   }
   n7 = Get skydiving_alpha()
   n8 = Anim Dynamics - Bone: top0_0(ComponentPose=Local To Component(LocalPose=Local Space Ref Pose()), Alpha=@n7, ExternalForce="0,0,0") {
    angularDampingOverride=1
    angularSpringConstant=50
    bGravityOverrideInSimSpace=true
    bOverrideAngularDamping=true
    gravityScale=-0.3
    simulationSpace=RootRelative
    alphaBoolBlend {blendOption=Linear}
    boundBone {boneName=top0_0}
    componentAppliedLinearAccClamp=<10000 10000 10000>
    componentLinearAccScale=<0.5 0.5 1>
    componentLinearVelScale=<0.5 0.5 1>
    physicsBodyDefinitions [{
      sphereCollisionRadius=0
      boundBone {boneName=top0_0}
      boxExtents=<10 30 30>
      constraintSetup {coneAngle=20, angularLimitsMax=<0 120 120>, angularLimitsMin=<0 -120 -120>, angularTarget=<1 0 0>}
      localJointOffset=<15 0 0>
     }]
   }
   n9 = Anim Dynamics - Bone: bottom0_0(ComponentPose=@n8, Alpha=@n7, ExternalForce="0,0,0") {
    angularDampingOverride=1
    angularSpringConstant=50
    bGravityOverrideInSimSpace=true
    bOverrideAngularDamping=true
    gravityScale=-0.3
    simulationSpace=RootRelative
    alphaBoolBlend {blendOption=Linear}
    boundBone {boneName=bottom0_0}
    componentAppliedLinearAccClamp=<10000 10000 10000>
    componentLinearAccScale=<0.5 0.5 1>
    componentLinearVelScale=<0.5 0.5 1>
    physicsBodyDefinitions [{
      sphereCollisionRadius=0
      boundBone {boneName=bottom0_0}
      boxExtents=<4 10 10>
      constraintSetup {coneAngle=20, angularLimitsMax=<0 120 120>, angularLimitsMin=<0 -120 -120>, angularTarget=<1 0 0>}
      localJointOffset=<5 0 0>
     }]
   }
   n10 = Anim Dynamics - Bone: long_world(ComponentPose=@n9, Alpha=@n7, ExternalForce="0,0,0") {
    alpha=-1
    angularDampingOverride=1
    angularSpringConstant=10
    bAngularSpring=true
    bGravityOverrideInSimSpace=true
    bOverrideAngularDamping=true
    bUsePlanarLimit=false
    bUseSphericalLimits=true
    gravityScale=-0.3
    simulationSpace=RootRelative
    alphaBoolBlend {blendOption=Linear}
    alphaScaleBiasClamp {scale=0.6}
    boundBone {boneName=long_world}
    componentAppliedLinearAccClamp=<10000 10000 10000>
    componentLinearAccScale=<0.3 0.3 0.2>
    componentLinearVelScale=<0.3 0.3 0.2>
    physicsBodyDefinitions [{
      sphereCollisionRadius=0
      boundBone {boneName=long_world}
      boxExtents=<30 30 10>
      constraintSetup {angularTargetAxis=AxisZ, coneAngle=20, twistAxis=AxisZ, angularLimitsMax=<70 70 10>, angularLimitsMin=<-70 -70 -10>, angularTarget=<0 0 1>}
      localJointOffset=<0 0 40>
     }]
    sphericalLimits [
     {limitRadius=18, drivingBone {boneName=pack}, sphereLocalOffset=<5 5 -3>}
     {limitRadius=25, drivingBone {boneName=pack}, sphereLocalOffset=<-10 20 -3>}
    ]
   }
   n11 = Anim Dynamics - Bone: fish2_0(ComponentPose=@n10, Alpha=1, ExternalForce="0,0,0") {
    angularDampingOverride=0.9999
    angularSpringConstant=20
    bAngularSpring=true
    bGravityOverrideInSimSpace=true
    bOverrideAngularDamping=true
    gravityScale=0.3
    simulationSpace=RootRelative
    alphaBoolBlend {blendOption=Linear}
    alphaScaleBiasClamp {scale=0.7}
    boundBone {boneName=fish2_0}
    componentLinearAccScale=<1 1 0.6>
    componentLinearVelScale=<1 1 0.6>
    physicsBodyDefinitions [{
      sphereCollisionRadius=0
      boundBone {boneName=fish2_0}
      boxExtents=<18 18 8>
      constraintSetup {angularLimitsMax=<0 70 70>, angularLimitsMin=<0 -70 -70>, angularTarget=<0 0 0>}
      localJointOffset=<0 0 -8>
     }]
   }
   n12 = Anim Dynamics - Bone: fish1_0(ComponentPose=@n11, Alpha=1, ExternalForce="0,0,0") {
    angularDampingOverride=0.9999
    angularSpringConstant=20
    bAngularSpring=true
    bGravityOverrideInSimSpace=true
    bOverrideAngularDamping=true
    gravityScale=0.3
    simulationSpace=RootRelative
    alphaBoolBlend {blendOption=Linear}
    alphaScaleBiasClamp {scale=0.7}
    boundBone {boneName=fish1_0}
    componentLinearAccScale=<1 1 0.6>
    componentLinearVelScale=<1 1 0.6>
    physicsBodyDefinitions [{
      sphereCollisionRadius=0
      boundBone {boneName=fish1_0}
      boxExtents=<18 18 8>
      constraintSetup {angularLimitsMax=<0 70 70>, angularLimitsMin=<0 -70 -70>, angularTarget=<0 0 0>}
      localJointOffset=<0 0 -8>
     }]
   }
   n13 = Anim Dynamics - Bone: fish3_0(ComponentPose=@n12, Alpha=1, ExternalForce="0,0,0") {
    angularDampingOverride=0.9999
    angularSpringConstant=20
    bAngularSpring=true
    bGravityOverrideInSimSpace=true
    bOverrideAngularDamping=true
    gravityScale=0.3
    simulationSpace=RootRelative
    alphaBoolBlend {blendOption=Linear}
    alphaScaleBiasClamp {scale=0.7}
    boundBone {boneName=fish3_0}
    componentLinearAccScale=<1 1 0.6>
    componentLinearVelScale=<1 1 0.6>
    physicsBodyDefinitions [{
      sphereCollisionRadius=0
      boundBone {boneName=fish3_0}
      boxExtents=<18 18 8>
      constraintSetup {angularLimitsMax=<0 70 70>, angularLimitsMin=<0 -70 -70>, angularTarget=<0 0 0>}
      localJointOffset=<0 0 -8>
     }]
   }
   n14 = "Constraint - top1_0 (R)"(ComponentPose=@n13, Alpha=0.77, ConstraintWeights_0=1) {
    alphaBoolBlend {blendOption=Linear}
    boneToModify {boneName=top1_0}
    constraintSetup [{transformType=Rotation, targetBone {boneName=top0_0}}]
    constraintWeights [1]
   }
   n15 = "Constraint - top2_0 (R)"(ComponentPose=@n14, Alpha=0.7, ConstraintWeights_0=1) {
    alphaBoolBlend {blendOption=Linear}
    boneToModify {boneName=top2_0}
    constraintSetup [{transformType=Rotation, targetBone {boneName=top0_0}}]
    constraintWeights [1]
   }
   n16 = "Constraint - top3_0 (R)"(ComponentPose=@n15, Alpha=0.74, ConstraintWeights_0=1) {
    alphaBoolBlend {blendOption=Linear}
    boneToModify {boneName=top3_0}
    constraintSetup [{transformType=Rotation, targetBone {boneName=top0_0}}]
    constraintWeights [1]
   }
   n17 = "Constraint - top4_0 (R)"(ComponentPose=@n16, Alpha=0.89, ConstraintWeights_0=1) {
    alphaBoolBlend {blendOption=Linear}
    boneToModify {boneName=top4_0}
    constraintSetup [{transformType=Rotation, targetBone {boneName=top0_0}}]
    constraintWeights [1]
   }
   n18 = "Constraint - top5_0 (R)"(ComponentPose=@n17, Alpha=0.75, ConstraintWeights_0=1) {
    alphaBoolBlend {blendOption=Linear}
    boneToModify {boneName=top5_0}
    constraintSetup [{transformType=Rotation, targetBone {boneName=top0_0}}]
    constraintWeights [1]
   }
   n19 = "Constraint - top6_0 (R)"(ComponentPose=@n18, Alpha=0.88, ConstraintWeights_0=1) {
    alphaBoolBlend {blendOption=Linear}
    boneToModify {boneName=top6_0}
    constraintSetup [{transformType=Rotation, targetBone {boneName=top0_0}}]
    constraintWeights [1]
   }
   n20 = "Constraint - top7_0 (R)"(ComponentPose=@n19, Alpha=0.87, ConstraintWeights_0=1) {
    alphaBoolBlend {blendOption=Linear}
    boneToModify {boneName=top7_0}
    constraintSetup [{transformType=Rotation, targetBone {boneName=top0_0}}]
    constraintWeights [1]
   }
   n21 = "Constraint - top8_0 (R)"(ComponentPose=@n20, Alpha=0.89, ConstraintWeights_0=1) {
    alphaBoolBlend {blendOption=Linear}
    boneToModify {boneName=top8_0}
    constraintSetup [{transformType=Rotation, targetBone {boneName=top0_0}}]
    constraintWeights [1]
   }
   n22 = "Constraint - top9_0 (R)"(ComponentPose=@n21, Alpha=0.73, ConstraintWeights_0=1) {
    alphaBoolBlend {blendOption=Linear}
    boneToModify {boneName=top9_0}
    constraintSetup [{transformType=Rotation, targetBone {boneName=top0_0}}]
    constraintWeights [1]
   }
   n23 = "Constraint - top10_0 (R)"(ComponentPose=@n22, Alpha=0.64, ConstraintWeights_0=1) {
    alphaBoolBlend {blendOption=Linear}
    boneToModify {boneName=top10_0}
    constraintSetup [{transformType=Rotation, targetBone {boneName=top0_0}}]
    constraintWeights [1]
   }
   n24 = "Constraint - top11_0 (R)"(ComponentPose=@n23, Alpha=0.93, ConstraintWeights_0=1) {
    alphaBoolBlend {blendOption=Linear}
    boneToModify {boneName=top11_0}
    constraintSetup [{transformType=Rotation, targetBone {boneName=top0_0}}]
    constraintWeights [1]
   }
   n25 = "Constraint - top12_0 (R)"(ComponentPose=@n24, Alpha=0.72, ConstraintWeights_0=1) {
    alphaBoolBlend {blendOption=Linear}
    boneToModify {boneName=top12_0}
    constraintSetup [{transformType=Rotation, targetBone {boneName=top0_0}}]
    constraintWeights [1]
   }
   n26 = "Constraint - top13_0 (R)"(ComponentPose=@n25, Alpha=0.8, ConstraintWeights_0=1) {
    alphaBoolBlend {blendOption=Linear}
    boneToModify {boneName=top13_0}
    constraintSetup [{transformType=Rotation, targetBone {boneName=top0_0}}]
    constraintWeights [1]
   }
   n27 = "Constraint - top14_0 (R)"(ComponentPose=@n26, Alpha=0.9, ConstraintWeights_0=1) {
    alphaBoolBlend {blendOption=Linear}
    boneToModify {boneName=top14_0}
    constraintSetup [{transformType=Rotation, targetBone {boneName=top0_0}}]
    constraintWeights [1]
   }
   n28 = "Constraint - top15_0 (R)"(ComponentPose=@n27, Alpha=0.71, ConstraintWeights_0=1) {
    alphaBoolBlend {blendOption=Linear}
    boneToModify {boneName=top15_0}
    constraintSetup [{transformType=Rotation, targetBone {boneName=top0_0}}]
    constraintWeights [1]
   }
   n29 = "Constraint - top16_0 (R)"(ComponentPose=@n28, Alpha=0.95, ConstraintWeights_0=1) {
    alphaBoolBlend {blendOption=Linear}
    boneToModify {boneName=top16_0}
    constraintSetup [{transformType=Rotation, targetBone {boneName=top0_0}}]
    constraintWeights [1]
   }
   n30 = "Constraint - top17_0 (R)"(ComponentPose=@n29, Alpha=0.71, ConstraintWeights_0=1) {
    alphaBoolBlend {blendOption=Linear}
    boneToModify {boneName=top17_0}
    constraintSetup [{transformType=Rotation, targetBone {boneName=top0_0}}]
    constraintWeights [1]
   }
   n31 = "Constraint - top18_0 (R)"(ComponentPose=@n30, Alpha=0.96, ConstraintWeights_0=1) {
    alphaBoolBlend {blendOption=Linear}
    boneToModify {boneName=top18_0}
    constraintSetup [{transformType=Rotation, targetBone {boneName=top0_0}}]
    constraintWeights [1]
   }
   n32 = "Constraint - top19_0 (R)"(ComponentPose=@n31, Alpha=0.7, ConstraintWeights_0=1) {
    alphaBoolBlend {blendOption=Linear}
    boneToModify {boneName=top19_0}
    constraintSetup [{transformType=Rotation, targetBone {boneName=top0_0}}]
    constraintWeights [1]
   }
   n33 = "Constraint - top20_0 (R)"(ComponentPose=@n32, Alpha=0.95, ConstraintWeights_0=1) {
    alphaBoolBlend {blendOption=Linear}
    boneToModify {boneName=top20_0}
    constraintSetup [{transformType=Rotation, targetBone {boneName=top0_0}}]
    constraintWeights [1]
   }
   n34 = "Constraint - top21_0 (R)"(ComponentPose=@n33, Alpha=0.91, ConstraintWeights_0=1) {
    alphaBoolBlend {blendOption=Linear}
    boneToModify {boneName=top21_0}
    constraintSetup [{transformType=Rotation, targetBone {boneName=top0_0}}]
    constraintWeights [1]
   }
   n35 = "Constraint - top22_0 (R)"(ComponentPose=@n34, Alpha=0.79, ConstraintWeights_0=1) {
    alphaBoolBlend {blendOption=Linear}
    boneToModify {boneName=top22_0}
    constraintSetup [{transformType=Rotation, targetBone {boneName=top0_0}}]
    constraintWeights [1]
   }
   n36 = "Constraint - top23_0 (R)"(ComponentPose=@n35, Alpha=0.7, ConstraintWeights_0=1) {
    alphaBoolBlend {blendOption=Linear}
    boneToModify {boneName=top23_0}
    constraintSetup [{transformType=Rotation, targetBone {boneName=top0_0}}]
    constraintWeights [1]
   }
   n37 = "Constraint - bottom1_0 (R)"(ComponentPose=@n36, Alpha=0.63, ConstraintWeights_0=1) {
    alphaBoolBlend {blendOption=Linear}
    boneToModify {boneName=bottom1_0}
    constraintSetup [{transformType=Rotation, targetBone {boneName=bottom0_0}}]
    constraintWeights [1]
   }
   n38 = "Constraint - bottom2_0 (R)"(ComponentPose=@n37, Alpha=0.63, ConstraintWeights_0=1) {
    alphaBoolBlend {blendOption=Linear}
    boneToModify {boneName=bottom2_0}
    constraintSetup [{transformType=Rotation, targetBone {boneName=bottom0_0}}]
    constraintWeights [1]
   }
   n39 = "Constraint - bottom3_0 (R)"(ComponentPose=@n38, Alpha=1, ConstraintWeights_0=1) {
    alphaBoolBlend {blendOption=Linear}
    boneToModify {boneName=bottom3_0}
    constraintSetup [{transformType=Rotation, targetBone {boneName=bottom0_0}}]
    constraintWeights [1]
   }
   n40 = "Constraint - bottom4_0 (R)"(ComponentPose=@n39, Alpha=0.69, ConstraintWeights_0=1) {
    alphaBoolBlend {blendOption=Linear}
    boneToModify {boneName=bottom4_0}
    constraintSetup [{transformType=Rotation, targetBone {boneName=bottom0_0}}]
    constraintWeights [1]
   }
   n41 = "Constraint - bottom5_0 (R)"(ComponentPose=@n40, Alpha=0.66, ConstraintWeights_0=1) {
    alphaBoolBlend {blendOption=Linear}
    boneToModify {boneName=bottom5_0}
    constraintSetup [{transformType=Rotation, targetBone {boneName=bottom0_0}}]
    constraintWeights [1]
   }
   n42 = "Constraint - bottom6_0 (R)"(ComponentPose=@n41, Alpha=0.87, ConstraintWeights_0=1) {
    alphaBoolBlend {blendOption=Linear}
    boneToModify {boneName=bottom6_0}
    constraintSetup [{transformType=Rotation, targetBone {boneName=bottom0_0}}]
    constraintWeights [1]
   }
   n43 = "Constraint - bottom7_0 (R)"(ComponentPose=@n42, Alpha=0.63, ConstraintWeights_0=1) {
    alphaBoolBlend {blendOption=Linear}
    boneToModify {boneName=bottom7_0}
    constraintSetup [{transformType=Rotation, targetBone {boneName=bottom0_0}}]
    constraintWeights [1]
   }
   n44 = "Constraint - bottom8_0 (R)"(ComponentPose=@n43, Alpha=0.82, ConstraintWeights_0=1) {
    alphaBoolBlend {blendOption=Linear}
    boneToModify {boneName=bottom8_0}
    constraintSetup [{transformType=Rotation, targetBone {boneName=bottom0_0}}]
    constraintWeights [1]
   }
   n45 = "Constraint - bottom9_0 (R)"(ComponentPose=@n44, Alpha=0.8, ConstraintWeights_0=1) {
    alphaBoolBlend {blendOption=Linear}
    boneToModify {boneName=bottom9_0}
    constraintSetup [{transformType=Rotation, targetBone {boneName=bottom0_0}}]
    constraintWeights [1]
   }
   n46 = "Constraint - bottom10_0 (R)"(ComponentPose=@n45, Alpha=0.74, ConstraintWeights_0=1) {
    alphaBoolBlend {blendOption=Linear}
    boneToModify {boneName=bottom10_0}
    constraintSetup [{transformType=Rotation, targetBone {boneName=bottom0_0}}]
    constraintWeights [1]
   }
   n47 = "Constraint - bottom11_0 (R)"(ComponentPose=@n46, Alpha=0.95, ConstraintWeights_0=1) {
    alphaBoolBlend {blendOption=Linear}
    boneToModify {boneName=bottom11_0}
    constraintSetup [{transformType=Rotation, targetBone {boneName=bottom0_0}}]
    constraintWeights [1]
   }
   n48 = "Constraint - bottom12_0 (R)"(ComponentPose=@n47, Alpha=0.68, ConstraintWeights_0=1) {
    alphaBoolBlend {blendOption=Linear}
    boneToModify {boneName=bottom12_0}
    constraintSetup [{transformType=Rotation, targetBone {boneName=bottom0_0}}]
    constraintWeights [1]
   }
   n49 = "Constraint - bottom13_0 (R)"(ComponentPose=@n48, Alpha=0.86, ConstraintWeights_0=1) {
    alphaBoolBlend {blendOption=Linear}
    boneToModify {boneName=bottom13_0}
    constraintSetup [{transformType=Rotation, targetBone {boneName=bottom0_0}}]
    constraintWeights [1]
   }
   n50 = "Constraint - bottom14_0 (R)"(ComponentPose=@n49, Alpha=0.68, ConstraintWeights_0=1) {
    alphaBoolBlend {blendOption=Linear}
    boneToModify {boneName=bottom14_0}
    constraintSetup [{transformType=Rotation, targetBone {boneName=bottom0_0}}]
    constraintWeights [1]
   }
   n51 = "Constraint - bottom15_0 (R)"(ComponentPose=@n50, Alpha=0.7, ConstraintWeights_0=1) {
    alphaBoolBlend {blendOption=Linear}
    boneToModify {boneName=bottom15_0}
    constraintSetup [{transformType=Rotation, targetBone {boneName=bottom0_0}}]
    constraintWeights [1]
   }
   n52 = "Constraint - bottom16_0 (R)"(ComponentPose=@n51, Alpha=0.7, ConstraintWeights_0=1) {
    alphaBoolBlend {blendOption=Linear}
    boneToModify {boneName=bottom16_0}
    constraintSetup [{transformType=Rotation, targetBone {boneName=bottom0_0}}]
    constraintWeights [1]
   }
   n53 = Control Rig(Source=Component To Local(ComponentPose=@n52)) {
    alphaBoolBlend {blendOption=Linear}
    controlRigAssetReference {blueprintRigClass=&/F_Med_Jellyfish_Backpack_CtrlRig.F_Med_Jellyfish_Backpack_CtrlRig_C}
   }
   n54 = Anim Dynamics - Bone: long0_2(ComponentPose=Local To Component(LocalPose=@n53), Alpha=@n1, ExternalForce="0,0,0") {
    angularDampingOverride=1
    angularSpringConstant=1000
    bAngularSpring=true
    bGravityOverrideInSimSpace=true
    bOverrideAngularDamping=true
    bUsePlanarLimit=false
    bUseSphericalLimits=true
    gravityScale=0
    alphaBoolBlend {blendOption=Linear}
    boundBone {boneName=long0_2}
    componentAppliedLinearAccClamp=<10000 10000 10000>
    componentLinearAccScale=<1 1 1>
    componentLinearVelScale=<1 1 1>
    physicsBodyDefinitions [{
      sphereCollisionRadius=0
      boundBone {boneName=long0_2}
      boxExtents=<50 50 20>
      constraintSetup {coneAngle=20, twistAxis=AxisZ, angularLimitsMax=<10 30 30>, angularLimitsMin=<-10 -30 -30>, angularTarget=<1 0 0>}
      localJointOffset=<10 0 0>
     }]
    relativeSpaceBone {boneName=root}
    sphericalLimits [
     {limitRadius=18, drivingBone {boneName=pack}, sphereLocalOffset=<5 5 -3>}
     {limitRadius=25, drivingBone {boneName=pack}, sphereLocalOffset=<-10 20 -3>}
    ]
   }
   n55 = Anim Dynamics - Bone: long0_3(ComponentPose=@n54, Alpha=@n1, ExternalForce="0,0,0") {
    angularDampingOverride=1
    angularSpringConstant=1000
    bAngularSpring=true
    bGravityOverrideInSimSpace=true
    bOverrideAngularDamping=true
    bUsePlanarLimit=false
    bUseSphericalLimits=true
    gravityScale=0
    alphaBoolBlend {blendOption=Linear}
    boundBone {boneName=long0_3}
    componentAppliedLinearAccClamp=<10000 10000 10000>
    componentLinearAccScale=<1 1 1>
    componentLinearVelScale=<1 1 1>
    physicsBodyDefinitions [{
      sphereCollisionRadius=0
      boundBone {boneName=long0_3}
      boxExtents=<50 50 20>
      constraintSetup {coneAngle=20, twistAxis=AxisZ, angularLimitsMax=<10 30 30>, angularLimitsMin=<-10 -30 -30>, angularTarget=<1 0 0>}
      localJointOffset=<10 0 0>
     }]
    relativeSpaceBone {boneName=root}
    sphericalLimits [
     {limitRadius=18, drivingBone {boneName=pack}, sphereLocalOffset=<5 5 -3>}
     {limitRadius=25, drivingBone {boneName=pack}, sphereLocalOffset=<-10 20 -3>}
    ]
   }
   n56 = Anim Dynamics - Bone: long1_2(ComponentPose=@n55, Alpha=@n1, ExternalForce="0,0,0") {
    angularDampingOverride=1
    angularSpringConstant=1000
    bAngularSpring=true
    bGravityOverrideInSimSpace=true
    bOverrideAngularDamping=true
    bUsePlanarLimit=false
    bUseSphericalLimits=true
    gravityScale=0
    alphaBoolBlend {blendOption=Linear}
    boundBone {boneName=long1_2}
    componentAppliedLinearAccClamp=<10000 10000 10000>
    componentLinearAccScale=<1 1 1>
    componentLinearVelScale=<1 1 1>
    physicsBodyDefinitions [{
      sphereCollisionRadius=0
      boundBone {boneName=long1_2}
      boxExtents=<50 50 20>
      constraintSetup {coneAngle=20, twistAxis=AxisZ, angularLimitsMax=<10 30 30>, angularLimitsMin=<-10 -30 -30>, angularTarget=<1 0 0>}
      localJointOffset=<10 0 0>
     }]
    relativeSpaceBone {boneName=root}
    sphericalLimits [
     {limitRadius=18, drivingBone {boneName=pack}, sphereLocalOffset=<5 5 -3>}
     {limitRadius=25, drivingBone {boneName=pack}, sphereLocalOffset=<-10 20 -3>}
    ]
   }
   n57 = Anim Dynamics - Bone: long1_3(ComponentPose=@n56, Alpha=@n1, ExternalForce="0,0,0") {
    angularDampingOverride=1
    angularSpringConstant=1000
    bAngularSpring=true
    bGravityOverrideInSimSpace=true
    bOverrideAngularDamping=true
    bUsePlanarLimit=false
    bUseSphericalLimits=true
    gravityScale=0
    alphaBoolBlend {blendOption=Linear}
    boundBone {boneName=long1_3}
    componentAppliedLinearAccClamp=<10000 10000 10000>
    componentLinearAccScale=<1 1 1>
    componentLinearVelScale=<1 1 1>
    physicsBodyDefinitions [{
      sphereCollisionRadius=0
      boundBone {boneName=long1_3}
      boxExtents=<50 50 20>
      constraintSetup {coneAngle=20, twistAxis=AxisZ, angularLimitsMax=<10 30 30>, angularLimitsMin=<-10 -30 -30>, angularTarget=<1 0 0>}
      localJointOffset=<10 0 0>
     }]
    relativeSpaceBone {boneName=root}
    sphericalLimits [
     {limitRadius=18, drivingBone {boneName=pack}, sphereLocalOffset=<5 5 -3>}
     {limitRadius=25, drivingBone {boneName=pack}, sphereLocalOffset=<-10 20 -3>}
    ]
   }
   n58 = Anim Dynamics - Bone: long2_2(ComponentPose=@n57, Alpha=@n1, ExternalForce="0,0,0") {
    angularDampingOverride=1
    angularSpringConstant=1000
    bAngularSpring=true
    bGravityOverrideInSimSpace=true
    bOverrideAngularDamping=true
    bUsePlanarLimit=false
    bUseSphericalLimits=true
    gravityScale=0
    alphaBoolBlend {blendOption=Linear}
    boundBone {boneName=long2_2}
    componentAppliedLinearAccClamp=<10000 10000 10000>
    componentLinearAccScale=<1 1 1>
    componentLinearVelScale=<1 1 1>
    physicsBodyDefinitions [{
      sphereCollisionRadius=0
      boundBone {boneName=long2_2}
      boxExtents=<50 50 20>
      constraintSetup {coneAngle=20, twistAxis=AxisZ, angularLimitsMax=<10 30 30>, angularLimitsMin=<-10 -30 -30>, angularTarget=<1 0 0>}
      localJointOffset=<10 0 0>
     }]
    relativeSpaceBone {boneName=root}
    sphericalLimits [
     {limitRadius=18, drivingBone {boneName=pack}, sphereLocalOffset=<5 5 -3>}
     {limitRadius=25, drivingBone {boneName=pack}, sphereLocalOffset=<-10 20 -3>}
    ]
   }
   n59 = Anim Dynamics - Bone: long2_3(ComponentPose=@n58, Alpha=@n1, ExternalForce="0,0,0") {
    angularDampingOverride=1
    angularSpringConstant=1000
    bAngularSpring=true
    bGravityOverrideInSimSpace=true
    bOverrideAngularDamping=true
    bUsePlanarLimit=false
    bUseSphericalLimits=true
    gravityScale=0
    alphaBoolBlend {blendOption=Linear}
    boundBone {boneName=long2_3}
    componentAppliedLinearAccClamp=<10000 10000 10000>
    componentLinearAccScale=<1 1 1>
    componentLinearVelScale=<1 1 1>
    physicsBodyDefinitions [{
      sphereCollisionRadius=0
      boundBone {boneName=long2_3}
      boxExtents=<50 50 20>
      constraintSetup {coneAngle=20, twistAxis=AxisZ, angularLimitsMax=<10 30 30>, angularLimitsMin=<-10 -30 -30>, angularTarget=<1 0 0>}
      localJointOffset=<10 0 0>
     }]
    relativeSpaceBone {boneName=root}
    sphericalLimits [
     {limitRadius=18, drivingBone {boneName=pack}, sphereLocalOffset=<5 5 -3>}
     {limitRadius=25, drivingBone {boneName=pack}, sphereLocalOffset=<-10 20 -3>}
    ]
   }
   n60 = Anim Dynamics - Bone: long3_2(ComponentPose=@n59, Alpha=@n1, ExternalForce="0,0,0") {
    angularDampingOverride=1
    angularSpringConstant=1000
    bAngularSpring=true
    bGravityOverrideInSimSpace=true
    bOverrideAngularDamping=true
    bUsePlanarLimit=false
    bUseSphericalLimits=true
    gravityScale=0
    alphaBoolBlend {blendOption=Linear}
    boundBone {boneName=long3_2}
    componentAppliedLinearAccClamp=<10000 10000 10000>
    componentLinearAccScale=<1 1 1>
    componentLinearVelScale=<1 1 1>
    physicsBodyDefinitions [{
      sphereCollisionRadius=0
      boundBone {boneName=long3_2}
      boxExtents=<50 50 20>
      constraintSetup {coneAngle=20, twistAxis=AxisZ, angularLimitsMax=<10 30 30>, angularLimitsMin=<-10 -30 -30>, angularTarget=<1 0 0>}
      localJointOffset=<10 0 0>
     }]
    relativeSpaceBone {boneName=root}
    sphericalLimits [
     {limitRadius=18, drivingBone {boneName=pack}, sphereLocalOffset=<5 5 -3>}
     {limitRadius=25, drivingBone {boneName=pack}, sphereLocalOffset=<-10 20 -3>}
    ]
   }
   n61 = Anim Dynamics - Bone: long3_3(ComponentPose=@n60, Alpha=@n1, ExternalForce="0,0,0") {
    angularDampingOverride=1
    angularSpringConstant=1000
    bAngularSpring=true
    bGravityOverrideInSimSpace=true
    bOverrideAngularDamping=true
    bUsePlanarLimit=false
    bUseSphericalLimits=true
    gravityScale=0
    alphaBoolBlend {blendOption=Linear}
    boundBone {boneName=long3_3}
    componentAppliedLinearAccClamp=<10000 10000 10000>
    componentLinearAccScale=<1 1 1>
    componentLinearVelScale=<1 1 1>
    physicsBodyDefinitions [{
      sphereCollisionRadius=0
      boundBone {boneName=long3_3}
      boxExtents=<50 50 20>
      constraintSetup {coneAngle=20, twistAxis=AxisZ, angularLimitsMax=<10 30 30>, angularLimitsMin=<-10 -30 -30>, angularTarget=<1 0 0>}
      localJointOffset=<10 0 0>
     }]
    relativeSpaceBone {boneName=root}
    sphericalLimits [
     {limitRadius=18, drivingBone {boneName=pack}, sphereLocalOffset=<5 5 -3>}
     {limitRadius=25, drivingBone {boneName=pack}, sphereLocalOffset=<-10 20 -3>}
    ]
   }
   n62 = Anim Dynamics - Bone: long4_2(ComponentPose=@n61, Alpha=@n1, ExternalForce="0,0,0") {
    angularDampingOverride=1
    angularSpringConstant=1000
    bAngularSpring=true
    bGravityOverrideInSimSpace=true
    bOverrideAngularDamping=true
    bUsePlanarLimit=false
    bUseSphericalLimits=true
    gravityScale=0
    alphaBoolBlend {blendOption=Linear}
    boundBone {boneName=long4_2}
    componentAppliedLinearAccClamp=<10000 10000 10000>
    componentLinearAccScale=<1 1 1>
    componentLinearVelScale=<1 1 1>
    physicsBodyDefinitions [{
      sphereCollisionRadius=0
      boundBone {boneName=long4_2}
      boxExtents=<50 50 20>
      constraintSetup {coneAngle=20, twistAxis=AxisZ, angularLimitsMax=<10 30 30>, angularLimitsMin=<-10 -30 -30>, angularTarget=<1 0 0>}
      localJointOffset=<10 0 0>
     }]
    relativeSpaceBone {boneName=root}
    sphericalLimits [
     {limitRadius=18, drivingBone {boneName=pack}, sphereLocalOffset=<5 5 -3>}
     {limitRadius=25, drivingBone {boneName=pack}, sphereLocalOffset=<-10 20 -3>}
    ]
   }
   n63 = Anim Dynamics - Bone: long4_3(ComponentPose=@n62, Alpha=@n1, ExternalForce="0,0,0") {
    angularDampingOverride=1
    angularSpringConstant=1000
    bAngularSpring=true
    bGravityOverrideInSimSpace=true
    bOverrideAngularDamping=true
    bUsePlanarLimit=false
    bUseSphericalLimits=true
    gravityScale=0
    alphaBoolBlend {blendOption=Linear}
    boundBone {boneName=long4_3}
    componentAppliedLinearAccClamp=<10000 10000 10000>
    componentLinearAccScale=<1 1 1>
    componentLinearVelScale=<1 1 1>
    physicsBodyDefinitions [{
      sphereCollisionRadius=0
      boundBone {boneName=long4_3}
      boxExtents=<50 50 20>
      constraintSetup {coneAngle=20, twistAxis=AxisZ, angularLimitsMax=<10 30 30>, angularLimitsMin=<-10 -30 -30>, angularTarget=<1 0 0>}
      localJointOffset=<10 0 0>
     }]
    relativeSpaceBone {boneName=root}
    sphericalLimits [
     {limitRadius=18, drivingBone {boneName=pack}, sphereLocalOffset=<5 5 -3>}
     {limitRadius=25, drivingBone {boneName=pack}, sphereLocalOffset=<-10 20 -3>}
    ]
   }
   Output Pose(Result=Component To Local(ComponentPose=@n63)) {layerGroup=None}
  ]}
 EventGraph {nodes [
   Event BlueprintUpdateAnimation()
   > Set skydiving_alpha(skydiving_alpha:double=SelectFloat(A=0, B=1, bPickA=Get bIsSkyDiving())) -> (Output_Get:double)
   TryGetPawnOwner()
  ]}
]
variables [skydiving_alpha {type=number}]
