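# Physics asset for the jellyfish backpack cosmetic, defining capsule collision bodies and joints for its pack and base bones.
# One large angled capsule covers the pack, with small capsules on the top, long and bottom base bones.
# Top and long bases joint directly to the pack with tight swing and twist limits.
# The bottom base joints tightly to a tiny bottom_world body, which rotates freely on all angular axes against the pack.
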PhysicsAsset /BRCosmetics/Accessories/FORT_Backpacks/F_MED_Jellyfish_Backpack/Meshes/F_MED_Jellyfish_Pack_Physics
bodies {
 bottom_base [{shapeType=Capsule, length=0.505, radius=5, center=<-2.0223 -3.6877 3.1522>, rotation=<rot 0 -90 0>}]
 bottom_world [{shapeType=Capsule, length=0.505, radius=0.505, center=<0 0 10>}]
 long_base [{shapeType=Capsule, length=0.505, radius=5, center=<-3.4863 1.6684 3.5619>, rotation=<rot 0 -90 0>}]
 pack [{shapeType=Capsule, length=11.9986, radius=10.2409, center=<-1.7467 -3.1092 -6.706>, rotation=<rot -18.7306 -90 34.0953>}]
 top_base [{shapeType=Capsule, length=0.505, radius=5, center=<-4.985 -1.3948 0.9107>, rotation=<rot 0 -90 0>}]
}
constraints [
 {bone1Name=bottom_base, bone2Name=bottom_world, swing1LimitDegrees=25, swing2LimitDegrees=25, twistLimitDegrees=5}
 {bone1Name=long_base, bone2Name=pack, swing1LimitDegrees=25, swing2LimitDegrees=25, twistLimitDegrees=5}
 {bone1Name=top_base, bone2Name=pack, swing1LimitDegrees=25, swing2LimitDegrees=25, twistLimitDegrees=5}
 {
  bone1Name=bottom_world
  bone2Name=pack
  swing1LimitDegrees=45
  swing1Motion=Free
  swing2LimitDegrees=45
  swing2Motion=Free
  twistLimitDegrees=45
  twistMotion=Free
 }
]
